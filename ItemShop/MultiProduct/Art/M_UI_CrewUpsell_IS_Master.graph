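# Translucent UI master material for the Crew upsell tile in the Item Shop.
# An IsFallback static switch picks either the ruleset-positioned OfferImage or the MF_BG_CrewUpsell background as color.
# The Crew wordmark is composited on top, scaled to fit with padding, nudged vertically and range-masked against edge streaking.
# Opacity uses the swipe mask, adding image/logo alpha and an optional gradient fade in fallback mode, then passes through rounded corners.
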Material /ItemShop/MultiProduct/Art/M_UI_CrewUpsell_IS_Master
graph {
 properties {
  bAutomaticallySetUsageInEditor=false
  bUsedWithStaticMesh=false
  blendMode=BLEND_Translucent
  materialDomain=MD_UI
  usedShadingModels=Default Lit
 }
 nodes [
  n1 = TextureObjectParameter() {group=Tile-driven params, parameterName=OfferImage, sortPriority=0, texture=/FortUILibrary/Input/PeripheralGuitar/Source/Empty1x1}
  n2 = TextureProperty(Tex=@n1)
  n3 = TextureSample(
   UVs=NamedRerouteDeclaration(None=MaterialFunctionCall!(Texture Size=@n2, UVs) {materialFunction=/FortUILibrary/Systems/ItemShopTile/Materials/MaterialFunctions/MF_UI_ItemShop_OfferImage_RulesetUVs}.UVs) {name=final UVs, nodeColor=<rgba 0.8281 0 1 1>}
   Tex=@n1
  ) {samplerSource=SSM_Clamp_WorldGroupSettings}
  n4 = TextureObject() {samplerType=SAMPLERTYPE_Alpha, texture=/Game/Athena/UI/Logos/Textures/T_Crew_Workmark_S}
  n5 = TextureProperty(Tex=@n4)
  n6 = Add(A=@n5, B=Constant2Vector() {desc=Horizontal/Vertical padding when Scaling to Fit, g=300, r=38})
  n7 = Add(
   A=MaterialFunctionCall(
    UVs=MaterialFunctionCall(TextureSize=@n6) {materialFunction=&MF_UI_UVScaling}.Scaled to Fit
    Texture Scale=Divide(A=@n5, B=@n6)
   ) {materialFunction=&ScaleUVsByCenter}.UVs
   B=Constant2Vector() {desc=Push logo up a bit, g=0.15}
  )
  n8 = TextureSample(UVs=@n7, Tex=@n4) {samplerSource=SSM_Clamp_WorldGroupSettings, texture=&/T_UI_CrewLogo}
  n9 = MaterialFunctionCall!(RangeMin, RangeMax, Value=@n7) {desc=Prevent streaks from the edges of the texture, materialFunction=/Engine/Functions/Engine_MaterialFunctions03/Math/WithinRangeFloat2}
  LinearInterpolate(
   A=StaticSwitchParameter(
    True=@n3.RGB
    False=MaterialFunctionCall!(CMSTexture=@n1) {materialFunction=/Game/Athena/UI/Subscriptions/Art/Materials/MF_BG_CrewUpsell}."Switch Param (False)"
   ) {parameterName=IsFallback}
   B=@n8.RGB
   Alpha=NamedRerouteDeclaration(None=Multiply(A=@n8.A, B=@n9.AllComponentsWithinRange)) {name=crew logo alpha, nodeColor=<rgba 0 1 0.7891 1>}
  )
  n10 = MaterialFunctionCall() {materialFunction=/FortUILibrary/Systems/ItemShopTile/Materials/MaterialFunctions/MF_UI_ItemShop_OfferImage_Swiping}
  n11 = Min(A=Max(A=NamedRerouteUsage(), B=NamedRerouteUsage()), B=@n10.Swipe alpha mask)
  n12 = MaterialFunctionCall(GradientStart=Constant() {r=1}, GradientEnd=Constant() {r=0.6}, RotationAngle=Constant() {r=-0.25}) {materialFunction=&MF_UI_LinearGradient}
  n13 = MaterialFunctionCall() {materialFunction=/FortUILibrary/Systems/ItemShopTile/Materials/MaterialFunctions/MF_UI_ItemShop_OfferImageParams}
  MaterialFunctionCall!(
   Input Alpha=StaticSwitchParameter(True=LinearInterpolate(A=@n11, B=@n12.Result, Alpha=@n13.Fade image alpha) {constAlpha=0}, False=@n10.Swipe alpha mask) {parameterName=IsFallback}
   bOverrideCornerRadius
   Corner Radius Override
  ) {materialFunction=/FortUILibrary/Systems/ItemShopTile/Materials/MaterialFunctions/MF_UI_ItemShop_RoundedCorners}
  NamedRerouteDeclaration(None=MaterialFunctionCall!(UVs=NamedRerouteUsage(), Alpha=@n3.A) {desc="prevents the texture from\r\nstretching at the edges", materialFunction=/Game/UI/Foundation/Items/ClampAlphaToTransparentBorder}.Result) {name=image alpha, nodeColor=<rgba 0.4062 0 1 1>}
 ]
}
