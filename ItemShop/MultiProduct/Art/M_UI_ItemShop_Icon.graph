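# Translucent UI material for an Item Shop icon, built from a signed distance field texture parameter.
# Texture UVs are scaled about the center, skewed by the SkewAngle parameter and sampled with clamping.
# An anti-aliasing function thresholds the distance field into the Icon Mask reroute.
# A Color parameter is separately dimmed by a multiply; neither branch shows a connection to a material output.
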Material /ItemShop/MultiProduct/Art/M_UI_ItemShop_Icon
graph {
 properties {
  bUsedWithStaticMesh=false
  blendMode=BLEND_Translucent
  materialDomain=MD_UI
  pixelDepthOffsetMode=PDOM_Legacy
  usedShadingModels=Default Lit
 }
 nodes [
  Multiply(A=VectorParameter() {parameterName=Color, defaultValue=<rgba 1 1 1 1>}.RGB) {constB=0.2}
  n1 = Constant() {r=2}
  n2 = TextureSampleParameter2D(UVs=MaterialFunctionCall!(
    UV=MaterialFunctionCall(Texture Scale=@n1) {materialFunction=&ScaleUVsByCenter}.UVs
    Skew Angle=VectorParameter() {parameterName=SkewAngle, defaultValue=<rgba 12 0 0 1>}.RGB
   ) {materialFunction=/FortUILibrary/MaterialFunctions/Basics/MF_UI_Skew}.Skewed UVs) {
   parameterName=Texture
   samplerSource=SSM_Clamp_WorldGroupSettings
   samplerType=SAMPLERTYPE_LinearColor
   texture=/FortUILibrary/Textures/Icons/T_UI_Icon_UIKit_Add_SDF
  }
  NamedRerouteDeclaration(None=MaterialFunctionCall(DistanceField=@n2.RGB, Threshold=Constant() {r=0.75}) {materialFunction=&MF_UI_AntiAliasing}.Outer) {name=Icon Mask, nodeColor=<rgba 0.9688 0 1 1>}
 ]
}
